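; Music encounter Blueprint for the AxisThrow boss, driving its music through combat phases of a single MetaSound.
; Idle state is hijacked to start music at phase 0, and combat switches to phase 1.
; Becoming DBNO during a specific encounter state switches to phase 2, then changes state when the audio finishes.
; Because the boss is convertible after defeat, conversion during that same state uninitializes the encounter, as does entering Completed.
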(event MusicEncounter|Boss|EventBossDBNOChanged
  (params (Boss%20Component%20Interface :type Music%20Encounter%20Component%20Boss%20Interface%20Interface) (Is%20DBNO :type Boolean))
  ;[0.5751 0.1986 0.8854 0.5] Since AxisThrow is convertable after defeat we have to make sure we track that and not just defeated/destroyed
  (bind _output_get (Variables|Default|SetIn%20Is%20DBNO Is%20DBNO))
  (if (and _output_get (MusicEncounter|GetMusicEncounterStateValue "NewEnumerator2"))
    (bind _combat_music_audio_component (Variables|MusicEncounter|Boss|Combat|GetCombat%20Music%20Audio%20Component)
      (:then
        (EventDispatchers|AssignOnAudioFinished _combat_music_audio_component (delegate Custom|OnAudioFinished_Event))
        (CallFunction|SetBossEncounterCombatPhase 2))
      (:else))))

(event Custom|OnAudioFinished_Event
  ;[0.5751 0.1986 0.8854 0.5] Since AxisThrow is convertable after defeat we have to make sure we track that and not just defeated/destroyed
  (CallFunction|SetBossEncounterState "NewEnumerator4"))

(event EventEnterCompletedState
  ;[0.5751 0.1986 0.8854 0.5] This is only called AFTER the combat music has finished playing. <----- See beside
  (|CallParentFunction)
  (CallFunction|UninitializeMusicEncounter))

(event EventEnterIdleState
  ;[0.5751 0.1986 0.8854 0.5] We are hijaking the Idle audio state to start the "combat" music right away. This is because we are only using 1 metasound
  (|CallParentFunction)
  (CallFunction|SetBossEncounterCombatPhase 0))

(event Custom|SetIsConverted
  (params (bIsConverted :type Boolean))
  (bind _output_get (Variables|Default|SetbIsConverted bIsConverted))
  (if (and _output_get (MusicEncounter|GetMusicEncounterStateValue "NewEnumerator2"))
    (CallFunction|UninitializeMusicEncounter)))

(event EventEnterCombatState
  ;[0.5751 0.1986 0.8854 0.5] When we enter combat lets switch to index 1, which is the combat music. There is only 1 phase. Otherwise we need to keep track of that elsewhere
  (|CallParentFunction)
  (CallFunction|SetBossEncounterCombatPhase 1))
