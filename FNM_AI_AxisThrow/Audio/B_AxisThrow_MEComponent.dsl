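; Music encounter component for the AxisThrow AI character (Doja), linking its pawn to the B_AxisThrow_ME music encounter.
; On BeginPlay it attaches a speaker sound to the owner's root, keeping a reference, so audio plays around the character in 2D.
; On initialization it caches the encounter and binds to the owner pawn's ConvertComponent.
; When the conversion event fires, it sets the encounter's IsConverted flag.
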(event EventBeginPlay
  (|CallParentFunction)
  ;[0.5751 0.1986 0.8854 0.5] We use this to attack an invisible speaker to Doja so that we play audio around her 2D at all times
  (bind _returnvalue (Audio|SpawnSoundAttached :Sound "/FNM_AI_AxisThrow/Audio/SpeakerSystem/SrcBs_AxisThrow_Speaker_Actor.SrcBs_AxisThrow_Speaker_Actor" :AttachToComponent (Variables|/Script/Engine.Actor|GetRootComponent (Components|GetOwner)) :AttachPointName "None" :LocationType "SnapToTarget" :bStopWhenAttachedToDestroyed true :VolumeMultiplier 1.0 :PitchMultiplier 1.0 :StartTime 0.0 :bAutoDestroy true))
  (Variables|Default|SetDojaSpeaker _returnvalue))

(event EventInitializeComponent
  (|CallParentFunction)
  (bind _asb_axis_throw_me (Utilities|Casting|CastToB_AxisThrow_ME (MusicEncounterComponent|GetMusicEncounter)))
  (Variables|Default|SetAs%20B%20Axis%20Throw%20ME _asb_axis_throw_me)
  (bind _asfort_player_pawn (Utilities|Casting|CastToFortPlayerPawn (Components|GetOwner)))
  (bind _convertcomponent (Variables|/Script/FortniteGame.FortPlayerPawn|GetConvertComponent _asfort_player_pawn)
    (:then
      (ConvertEvents|AssignOnConvertedEvent _convertcomponent (delegate Custom|OnConvertedEvent_Event)))
    (:else)))

(event Custom|OnConvertedEvent_Event
  (params (InstigatorPawn :type Fort%20Pawn%20Object%20Reference) (ConvertedPawn :type Fort%20Pawn%20Object%20Reference))
  (Class|BAxisThrowME|SetIsConverted (Variables|Default|GetAs%20B%20Axis%20Throw%20ME) true))
