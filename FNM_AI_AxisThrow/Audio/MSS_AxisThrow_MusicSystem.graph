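# Stereo MetaSound music system for the AxisThrow boss encounter, driven by boss phase and tempo.
# Phase changes trigger metronome-quantized one-shots from a phase-indexed array, then beat-delayed crossfades between two phase loops, releasing both at the last phase.
# Distance high-passes the phase 0 loop, and a weapon-fire envelope modulates a 100 Hz EQ.
# Output also feeds a combat speaker bus; pause fade-out or the final one-shot finishes playback.
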MetaSoundSource /FNM_AI_AxisThrow/Audio/MSS_AxisThrow_MusicSystem
inputs [
 MusicEncounter.Attenuation.Distance {type=Float}
 MusicEncounter.Attenuation.Volume.FalloffDistance {type=Float}
 MusicEncounter.Attenuation.Volume.InnerRadius {type=Float}
 MusicEncounter.bIsPaused {type=Bool}
 MusicEncounter.Boss.Phase {type=Int32}
 MusicEncounter.PauseFadeOutTime {type=Time}
 Tempo {type=Float, access=Constructor}
 Time.ElapsedTime {type=Float}
 Time.GameTimeSeconds {type=Float}
 Time.PlayTime {type=Float}
 Time.TimeSinceLastPlayed {type=Float}
 UE.Source.OnPlay {type=Trigger}
]
interfaces [Fort.Time:1.0, UE.OutputFormat.Stereo:1.0, UE.Source.OneShot:1.0, UE.Source:1.1]
outputs [
 "UE.OutputFormat.Stereo.Audio:0" {type=Audio}
 "UE.OutputFormat.Stereo.Audio:1" {type=Audio}
 UE.Source.OneShot.OnFinished {type=Trigger}
]
pages [Default {
  input_defaults {
   MusicEncounter.Attenuation.Distance=0
   MusicEncounter.Attenuation.Volume.FalloffDistance=10000
   MusicEncounter.Attenuation.Volume.InnerRadius=2500
   MusicEncounter.Boss.Phase=0
   MusicEncounter.PauseFadeOutTime=3
   MusicEncounter.bIsPaused=null
   Tempo=140
   Time.ElapsedTime=0
   Time.GameTimeSeconds=0
   Time.PlayTime=0
   Time.TimeSinceLastPlayed=500
   UE.Source.OnPlay=null
  }
  nodes [
   n1 = UE.BPMToSeconds(BPM=^Tempo) {also {defaults {Divisions of Whole Note=1}}}
   n2 = UE.BPMToSeconds(BPM=^Tempo)
   n3 = HarmonixNodes.TriggerToTransport(Play=^UE.Source.OnPlay)
   TriggerOnValueChange_1 = UE.Trigger On Value Change:Int32(Value=^MusicEncounter.Boss.Phase)
   TriggerAnyTriggerAny2_0 = "TriggerAny.Trigger Any (2)"(In 1=^UE.Source.OnPlay, In 0=@TriggerOnValueChange_1.Trigger)
   TriggerOnce_0 = UE.Trigger Once(
    Trigger In=HarmonixNodes.MidiClockSubdivisionTrigger(MIDI Clock=HarmonixNodes.Metronome(Tempo=^Tempo, Transport=@n3.Transport).MIDI Clock).Trigger Out
    Reset=UE.Reroute(Value=@TriggerAnyTriggerAny2_0.Out).Value
   )
   n4 = UE.Reroute(Value=@TriggerOnce_0.Trigger Out)
   TriggerDelay_2 = UE.Trigger Delay(Delay Time=UE.BPMToSeconds(BPM=^Tempo).Seconds, In=@n4.Value)
   TriggerOnce_1 = UE.Trigger Once(Trigger In=@n4.Value, Reset=@TriggerDelay_2.Out)
   n5 = VariableDeferredAccessor.Trigger(Variable=InitVariable.Trigger().Variable)
   TriggerOnce_2 = UE.Trigger Once(Trigger In=@TriggerOnce_1.Trigger Out, Reset=@n5.Value)
   n6 = VariableAccessor.WaveAsset:Array(Variable=InitVariable.WaveAsset:Array().Variable)
   n7 = UE.Subtract:Int32(PrimaryOperand=Array.Num:WaveAsset:Array(Array=VariableAccessor.WaveAsset:Array(Variable=@n6.Variable).Value).Num) {also {defaults {AdditionalOperands=1}}}
   n8 = VariableAccessor.Int32(Variable=VariableMutator.Int32(Variable=InitVariable.Int32().Variable, Value=Value.Value:Int32(Set=^UE.Source.OnPlay, Target Value=@n7.Out).Output Value).Variable)
   n9 = TriggerCompare.Clamp:Int32(A=^MusicEncounter.Boss.Phase, Compare=@TriggerOnce_2.Trigger Out, B=@n8.Value)
   TriggerRouteTriggerRouteFloat2_0 = "TriggerRoute.Trigger Route (Float, 2):Float"(Set 1=@n9.False, Set 0=@n9.True) {also {defaults {Value 0=1, Value 1=3.5}}}
   TriggerDelay_1 = UE.Trigger Delay(
    Delay Time=UE.BPMToSeconds(BPM=^Tempo, Beat Multiplier=@TriggerRouteTriggerRouteFloat2_0.Value).Seconds
    In=@TriggerRouteTriggerRouteFloat2_0.On Set
   )
   n10 = VariableAccessor.Int32(Variable=@n8.Variable)
   n11 = TriggerCompare.Clamp:Int32(A=^MusicEncounter.Boss.Phase, Compare=@TriggerDelay_1.Out, B=VariableAccessor.Int32(Variable=@n10.Variable).Value) {also {defaults {Type=1}}}
   n12 = UE.Reroute(Value=UE.Reroute(Value=@n11.False).Value)
   n13 = TriggerCompare.Clamp:Int32(A=^MusicEncounter.Boss.Phase, Compare=@n11.True)
   n14 = VariableDeferredAccessor.Trigger(Variable=InitVariable.Trigger().Variable)
   WavePlayer_0 = UE.Wave Player:Stereo(
    Play=UE.Reroute(Value=TriggerCompare.Clamp:Int32(A=^MusicEncounter.Boss.Phase, Compare=@n13.False) {also {defaults {B=1}}}.True).Value
    Stop=@n14.Value
   ) {also {
     defaults {Loop="<MetaSoundBuilderResult.SUCCEEDED: 0>", Wave Asset=/FNM_AI_AxisThrow/Audio/Music/Final/MX_S37_IP_Combat_AxisThrow_Phase_1_LP}
     version [1,1]
    }}
   TriggerAnyTriggerAny2_2 = "TriggerAny.Trigger Any (2)"(In 0=@n12.Value, In 1=@WavePlayer_0.On Play)
   n15 = VariableDeferredAccessor.Trigger(Variable=InitVariable.Trigger().Variable)
   WavePlayer_1 = UE.Wave Player:Stereo(Play=@n13.True, Stop=@n15.Value) {also {
     defaults {Loop="<MetaSoundBuilderResult.SUCCEEDED: 0>", Wave Asset=/FNM_AI_AxisThrow/Audio/Music/Final/MX_S37_IP_Combat_AxisThrow_Phase_0_LP}
     version [1,1]
    }}
   ADSREnvelopeADSREnvelope_0 = ADSR Envelope.ADSR Envelope:Float(
    Release Time=@n1.Seconds
    Attack Time=@n2.Seconds
    Trigger Release=UE.Reroute(Value=@TriggerAnyTriggerAny2_2.Out).Value
    Trigger Attack=UE.Reroute(Value=UE.Reroute(Value=@WavePlayer_1.On Play).Value).Value
   ) {also {defaults {Sustain Level=1}}}
   TriggerAnyTriggerAny2_3 = "TriggerAny.Trigger Any (2)"(In 1=UE.Reroute(Value=@n12.Value).Value, In 0=@WavePlayer_1.On Play)
   ADSREnvelopeADSREnvelope_1 = ADSR Envelope.ADSR Envelope:Float(Release Time=@n1.Seconds, Attack Time=@n2.Seconds, Trigger Release=@TriggerAnyTriggerAny2_3.Out, Trigger Attack=@WavePlayer_0.On Play) {also {defaults {Sustain Level=1}}}
   DecibelstoLinearGain_1 = UE.Decibels to Linear Gain:Float()
   n16 = MapRange.MapRange:Float(
    In Range B=^MusicEncounter.Attenuation.Volume.FalloffDistance
    In Range A=^MusicEncounter.Attenuation.Volume.InnerRadius
    In=UE.InterpTo:Audio(Target=^MusicEncounter.Attenuation.Distance) {also {defaults {Interp Time=0.5}}}.Value
   ) {also {defaults {Out Range A=250, Out Range B=600}}}
   OnePoleHighPassFilter_0 = UE.One-Pole High Pass Filter:Audio(Cutoff Frequency=@n16.Out Value, In=@WavePlayer_1.Out Left)
   OnePoleHighPassFilter_1 = UE.One-Pole High Pass Filter:Audio(Cutoff Frequency=@n16.Out Value, In=@WavePlayer_1.Out Right)
   n17 = UE.Reroute(Value=UE.Reroute(Value=@TriggerOnce_1.Trigger Out).Value)
   WavePlayer_2 = UE.Wave Player:Stereo(Wave Asset=Array.Get:WaveAsset:Array(Index=^MusicEncounter.Boss.Phase, Trigger=@n17.Value, Array=@n6.Value).Element, Play=@n17.Value) {also {version [1,1]}}
   AudioMixerAudioMixerStereo3_0 = "AudioMixer.Audio Mixer (Stereo, 3)"(
    Gain 1=@ADSREnvelopeADSREnvelope_0.Out Envelope
    Gain 2=@ADSREnvelopeADSREnvelope_1.Out Envelope
    Gain 0=@DecibelstoLinearGain_1.Linear Gain
    In 1 L=@OnePoleHighPassFilter_0.Out
    In 1 R=@OnePoleHighPassFilter_1.Out
    In 2 L=@WavePlayer_0.Out Left
    In 2 R=@WavePlayer_0.Out Right
    In 0 L=@WavePlayer_2.Out Left
    In 0 R=@WavePlayer_2.Out Right
   )
   n18 = MapRange.MapRange:Float(In=Modulation.GetModulatorValue() {also {defaults {Modulator=/Game/Sounds/Modulation/EnvelopeFollowers/FortWeapon/CB_WeaponFireEnvFollower}}}.Out) {also {defaults {Out Range A=-3, Out Range B=0}}}
   BiquadFilter_1 = UE.Biquad Filter:Audio(In=@AudioMixerAudioMixerStereo3_0.Out R, Gain=@n18.Out Value) {also {defaults {Bandwidth=2, Cutoff Frequency=100, Type=4}}}
   DecibelstoLinearGain_0 = UE.Decibels to Linear Gain:Float() {also {defaults {Decibels=3}}}
   TriggerOnValueChange_0 = UE.Trigger On Value Change:Bool(Value=^MusicEncounter.bIsPaused)
   n19 = TriggerCompare.Clamp:Bool(A=^MusicEncounter.bIsPaused, Compare=@TriggerOnValueChange_0.Trigger) {also {defaults {B="<MetaSoundBuilderResult.SUCCEEDED: 0>"}}}
   n20 = Value.Value:Float(Reset=@n19.False, Set=@n19.True) {also {defaults {Init Value=1}}}
   n21 = UE.Multiply:Float(
    AdditionalOperands=@DecibelstoLinearGain_0.Linear Gain
    PrimaryOperand=UE.InterpTo:Audio(Interp Time=^MusicEncounter.PauseFadeOutTime, Target=@n20.Output Value).Value
   )
   Multiply_0 = UE.Multiply:Audio by Float(PrimaryOperand=@BiquadFilter_1.Out, AdditionalOperands=@n21.Out)
   BiquadFilter_0 = UE.Biquad Filter:Audio(In=@AudioMixerAudioMixerStereo3_0.Out L, Gain=@n18.Out Value) {also {defaults {Bandwidth=2, Cutoff Frequency=100, Type=4}}}
   Multiply_1 = UE.Multiply:Audio by Float(PrimaryOperand=@BiquadFilter_0.Out, AdditionalOperands=@n21.Out)
   AudioBusWriter2_0 = "UE.Audio Bus Writer (2)"(In 1=@Multiply_0.Out, In 0=@Multiply_1.Out) {also {defaults {Audio Bus=/FNM_AI_AxisThrow/Audio/SpeakerSystem/AB_AxisThrow_CombatSpeaker}}}
   InitVariable.WaveAsset:Array()
   InitVariable.WaveAsset:Array()
   TriggerDelay_0 = UE.Trigger Delay(Delay Time=^MusicEncounter.PauseFadeOutTime, Reset=@n20.On Reset, In=@n20.On Set)
   TriggerAnyTriggerAny2_1 = "TriggerAny.Trigger Any (2)"(
    In 0=TriggerCompare.Clamp:Int32(A=^MusicEncounter.Boss.Phase, B=@n10.Value, Compare=@WavePlayer_2.On Finished).True
    In 1=@TriggerDelay_0.Out
   )
   VariableMutator.Trigger(Value=@ADSREnvelopeADSREnvelope_1.On Done, Variable=@n14.Variable)
   VariableMutator.Trigger(Value=@TriggerDelay_1.Out, Variable=@n5.Variable)
   VariableMutator.Trigger(Value=@ADSREnvelopeADSREnvelope_0.On Done, Variable=@n15.Variable)
  ]
  links [
   {dst="@UE.OutputFormat.Stereo.Audio:1", src=Multiply_0.Out}
   {dst="@UE.OutputFormat.Stereo.Audio:0", src=Multiply_1.Out}
   {dst="@UE.Source.OneShot.OnFinished", src=TriggerAnyTriggerAny2_1.Out}
  ]
 }]
properties {
 assetImportData=&self:AssetImportData
 bEnableBaseSubmix=false
 bEnableSubmixSends=false
 numChannels=2
 outputFormat=Stereo
 sampleRate=48000
 soundClassObject=/Game/Sounds/FortSoundClassesAndModes/Fort_Music
 assetUserData [&self:AudioAssetUserData_0]
 busSends [{audioBus=&AB_Sidechain_FortAmbient_Vol, sendLevel=0.5}]
}
